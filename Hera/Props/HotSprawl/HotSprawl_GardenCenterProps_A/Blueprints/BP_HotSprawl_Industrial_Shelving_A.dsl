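; Industrial shelving prop for the HotSprawl garden center that styles its secondary tool barrel and flour bag stack meshes from building texture data.
; After the parent construction script, the first texture entry drives the tool barrel's tint, with a hardcoded clean digging-tools material also involved.
; The second entry gives the bag stack its override material when valid, otherwise its tint color.
; BeginPlay, overlap and Tick events are empty.
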(event EventBeginPlay)

(event Collision|EventActorBeginOverlap
  (params (OtherActor :type Actor%20Object%20Reference)))

(event EventTick
  (params (DeltaSeconds :type Float%20%28single-precision%29)))

(fn ConstructionScript
  (bind _sm_outdoor_toolbarrel_a (Variables|Default|GetSM_Outdoor_ToolBarrel_A))
  (bind _sm_container_flourbagstack_a (Variables|Default|GetSM_Container_FlourBagStack_A))
  (|CallParentFunction)
  (bind _outtexturedataarray (Building|GetBuildingTextureData))
  (if (Utilities|Array|IsValidIndex _outtexturedataarray)
    (Default|SetColoronSecondaryMesh (Math|Conversions|ToLinearColor(Color) (Variables|/Script/FortniteGame.BuildingTextureData|GetTintColor (Utilities|Array|Get(acopy) _outtexturedataarray))) _sm_outdoor_toolbarrel_a)
    (Rendering|Material|SetMaterial _sm_outdoor_toolbarrel_a 0 "/Game/Environments/Asteria/Props/Outdoor/Outdoor_DiggingTools_A/Materials/MI_Outdoor_DiggingToolsClean_A_A.MI_Outdoor_DiggingToolsClean_A_A")
    (bind _outtexturedataarray_2 (Building|GetBuildingTextureData))
    (if (Utilities|Array|IsValidIndex _outtexturedataarray_2 1)
      (if (Utilities|IsValid (Variables|/Script/FortniteGame.BuildingTextureData|GetOverrideMaterial (Utilities|Array|Get(acopy) _outtexturedataarray_2 1)) :!class /Script/Engine.KismetSystemLibrary)
        (Rendering|Material|SetMaterial _sm_container_flourbagstack_a 0 (Variables|/Script/FortniteGame.BuildingTextureData|GetOverrideMaterial (Utilities|Array|Get(acopy) _outtexturedataarray_2 1)))
        (Default|SetColoronSecondaryMesh (Math|Conversions|ToLinearColor(Color) (Variables|/Script/FortniteGame.BuildingTextureData|GetTintColor (Utilities|Array|Get(acopy) _outtexturedataarray_2 1))) _sm_container_flourbagstack_a)))))
