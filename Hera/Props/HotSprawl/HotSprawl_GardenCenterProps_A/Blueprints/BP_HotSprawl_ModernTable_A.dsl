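; BP_HotSprawl_ModernTable_A is a HotSprawl garden center prop Blueprint for a modern table, with a potted plant mesh component.
; Its construction script runs the parent script, then passes the small potted plant component to SetColoronSecondaryMesh with a fixed pale yellow color.
; The BeginPlay, overlap and Tick events are empty, so the only logic in this graph is that construction-time tint.
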(event EventBeginPlay)

(event Collision|EventActorBeginOverlap
  (params (OtherActor :type Actor%20Object%20Reference)))

(event EventTick
  (params (DeltaSeconds :type Float%20%28single-precision%29)))

(fn ConstructionScript
  (|CallParentFunction)
  (Default|SetColoronSecondaryMesh "(R=0.775822,G=0.745404,B=0.162029,A=1.000000)" (Variables|Default|GetSM_Outdoor_SmallPottedPlant_E)))
