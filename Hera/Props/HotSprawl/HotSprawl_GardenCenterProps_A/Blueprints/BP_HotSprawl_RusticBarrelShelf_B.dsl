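; Garden-center rustic barrel shelf prop from HotSprawl that colors its flour bags and potted plants at construction time.
; When building texture data supplies a valid override material, the construction script applies it and its tint color to three flour bag meshes.
; Within that same branch, AddColorToFlower colors four potted plants using three index variables into the Flower Colors array, capping too-high indices at the last entry.
; BeginPlay, overlap and Tick events are empty.
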(event EventBeginPlay)

(event Collision|EventActorBeginOverlap
  (params (OtherActor :type Actor%20Object%20Reference)))

(event EventTick
  (params (DeltaSeconds :type Float%20%28single-precision%29)))

(fn AddColorToFlower
  (params (Static%20Mesh :type Static%20Mesh%20Component%20Object%20Reference) (Color%20Index :type Integer) (Flower%20Colors :type Array%20of%20Linear%20Color%20Structures))
  (bind _returnvalue (Utilities|Array|Length Flower%20Colors))
  (bind _returnvalue_2 (- _returnvalue 1))
  (if (<= Color%20Index _returnvalue_2)
    (Default|SetColoronSecondaryMesh (Utilities|Array|Get(aref) Flower%20Colors Color%20Index) Static%20Mesh)
    (else
      (Default|SetColoronSecondaryMesh (Utilities|Array|Get(aref) Flower%20Colors _returnvalue_2) Static%20Mesh))))

(fn ConstructionScript
  (|CallParentFunction)
  (bind _outtexturedataarray (Building|GetBuildingTextureData))
  (if (Utilities|Array|IsValidIndex _outtexturedataarray)
    (if (Utilities|IsValid (Variables|/Script/FortniteGame.BuildingTextureData|GetOverrideMaterial (Utilities|Array|Get(acopy) _outtexturedataarray)) :!class /Script/Engine.KismetSystemLibrary)
      (for _array_element (Utilities|Array|MakeArray (Variables|Default|GetSM_Container_FlourBag_A) (Variables|Default|GetSM_Container_FlourBag_A1) (Variables|Default|GetSM_Container_FlourBag_A2))
        (Rendering|Material|SetMaterial _array_element 0 (Variables|/Script/FortniteGame.BuildingTextureData|GetOverrideMaterial (Utilities|Array|Get(acopy) _outtexturedataarray)))
        (Default|SetColoronSecondaryMesh (Math|Conversions|ToLinearColor(Color) (Variables|/Script/FortniteGame.BuildingTextureData|GetTintColor (Utilities|Array|Get(acopy) _outtexturedataarray))) _array_element)
        (for _array_element_2 (Utilities|Array|MakeArray (Variables|Default|GetSM_Outdoor_SmallPottedPlant_A1) (Variables|Default|GetSM_Outdoor_SmallPottedPlant_C1))
          (CallFunction|AddColortoFlower _array_element_2 (Variables|Default|GetFlower%20Color%201) (Variables|Default|GetFlower%20Colors))
          (CallFunction|AddColortoFlower (Variables|Default|GetSM_Outdoor_SmallPottedPlant_A) (Variables|Default|GetFlower%20Color%202) (Variables|Default|GetFlower%20Colors))
          (CallFunction|AddColortoFlower (Variables|Default|GetSM_Outdoor_SmallPottedPlant_C) (Variables|Default|GetFlower%20Color%203) (Variables|Default|GetFlower%20Colors)))))))
