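; Rustic barrel shelf prop for the HotSprawl garden center set, whose flour bags can be reskinned through Fortnite building texture data.
; The construction script calls the parent, then checks that the first texture data entry exists and has a valid override material.
; It applies that material to slot 0 of both flour bag meshes and passes the entry's tint color to SetColoronSecondaryMesh.
; BeginPlay, overlap, Tick and CreateMIDFromTextureData are empty.
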(event EventBeginPlay)

(event Collision|EventActorBeginOverlap
  (params (OtherActor :type Actor%20Object%20Reference)))

(event EventTick
  (params (DeltaSeconds :type Float%20%28single-precision%29)))

(fn CreateMIDFromTextureData
  (params (SwatchIndex :type Integer) (Mesh%20to%20Modify :type Primitive%20Component%20Object%20Reference)))

(fn ConstructionScript
  (|CallParentFunction)
  (bind _outtexturedataarray (Building|GetBuildingTextureData))
  ;[1 1 1] Validation - for TD Index & Material
  (if (Utilities|Array|IsValidIndex _outtexturedataarray)
    (if (Utilities|IsValid (Variables|/Script/FortniteGame.BuildingTextureData|GetOverrideMaterial (Utilities|Array|Get(acopy) _outtexturedataarray)) :!class /Script/Engine.KismetSystemLibrary)
      ;[1 1 1] This shows which SMs will have TD Data Applied
      (for _array_element (Utilities|Array|MakeArray (Variables|Default|GetSM_Container_FlourBag_A) (Variables|Default|GetSM_Container_FlourBag_B))
        (Rendering|Material|SetMaterial _array_element 0 (Variables|/Script/FortniteGame.BuildingTextureData|GetOverrideMaterial (Utilities|Array|Get(acopy) _outtexturedataarray)))
        (Default|SetColoronSecondaryMesh (Math|Conversions|ToLinearColor(Color) (Variables|/Script/FortniteGame.BuildingTextureData|GetTintColor (Utilities|Array|Get(acopy) _outtexturedataarray))) _array_element)))))
